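# Base material for the CopperMoggy Prime pickaxe, adding an animated "Ranked Reload" grid effect on skeletal meshes.
# Between the 2023 character parent PreFX and PostFX functions, an animated grid function drives color, emissive, roughness and specular.
# A repeating radial wipe, measured in pre-skinned local space and warped by Perlin noise and grid, tints the emissive.
# A channel mask times the character skin's digital mosaic noise blends the effect in.
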Material /BRCosmetics/Weapons/FORT_Melee/Pickaxe_CopperMoggy_Swan/Skins/Prime/Materials/M_Pickaxe_CopperMoggy_Prime_Base
graph {
 properties {
  bUseMaterialAttributes=true
  bUsedWithMeshDeformer=true
  bUsedWithSkeletalMesh=true
  bUsedWithStaticMesh=false
  pixelDepthOffsetMode=PDOM_Legacy
  shadingModel=MSM_FromMaterialExpression
  translucencyLightingMode=TLM_SurfacePerPixelLighting
  usedShadingModels=Default Lit | Subsurface | Clear Coat | Subsurface Profile | Cloth
 }
 nodes [
  n1 = MaterialFunctionCall() {materialFunction=&MF_CharacterParent_2023_PreFX}
  MaterialFunctionCall(Material=NamedRerouteUsage(), M=@n1.M, Pre Skinned Local Position=@n1.M, Distance Blend=@n1.M, Fresnel=@n1.M) {materialFunction=&MF_CharacterParent_2023_PostFX}
  NamedRerouteDeclaration(None=@n1.Result) {name=IN_Material, nodeColor=<rgba 0.875 1 0 1>}
  NamedRerouteDeclaration(None=@n1.Tech Art Mask) {name=TechArtMask, nodeColor=<rgba 0.8047 0 1 1>}
  n2 = TextureSampleParameter2D() {
   group=Ranked Reload
   parameterName=RankedReload.GridTexture
   samplerType=SAMPLERTYPE_Masks
   sortPriority=0
   texture=/Game/Global/Textures/Default/Blanks/T_White
  }
  n3 = NamedRerouteUsage()
  n4 = VectorParameter() {group=Ranked Reload, parameterName=RankedReload.Color_01, sortPriority=2, defaultValue=<rgba 0.05 0.25 0.35 1>}
  n5 = VectorParameter() {group=Ranked Reload, parameterName=RankedReload.Color_02, sortPriority=3, defaultValue=<rgba 0 0.65 1 1>}
  n6 = VectorParameter() {group=Ranked Reload, parameterName=RankedReload.LineColor, sortPriority=4, defaultValue=<rgba 0 0.75 1.5 1>}
  n7 = MaterialFunctionCall!(
   Grid=@n2.R
   Grid Lines=NamedRerouteDeclaration(None=ChannelMaskParameter(None=@n3) {group=Ranked Reload, parameterName=RankedReload.GridLineChannel, sortPriority=1}.R) {name=GridLines, nodeColor=<rgba 1 0.3281 0 1>}
   Color 01=@n4.RGB
   Color 02=@n5.RGB
   Speed=ScalarParameter() {defaultValue=0.2, group=Ranked Reload, parameterName=RankedReload.Speed, sortPriority=5}
   Speed Random Amount=ScalarParameter() {defaultValue=2, group=Ranked Reload, parameterName=RankedReload.Speed_RandomAmount, sortPriority=6}
   Line Color=@n6.RGB
  ) {materialFunction=/Game/Global/MaterialFunctions/Cosmetics/AnimatedGrid/MF_AnimatedGridTexture}
  NamedRerouteDeclaration(None=@n7.Grayscale Grid) {name=GrayscaleGrid, nodeColor=<rgba 0.0312 0 1 1>}
  n8 = NamedRerouteUsage()
  n9 = VectorParameter() {group=Ranked Reload, parameterName=RankedReload.Wipe_Color, defaultValue=<rgba 0.8594 2 5 1>}
  n10 = Multiply(A=NamedRerouteUsage(), B=ScalarParameter() {defaultValue=0.25, group=Ranked Reload, parameterName=RankedReload.WipeEmissive_Strength})
  n11 = ScalarParameter() {defaultValue=0.85, group=Ranked Reload, parameterName=RankedRelaoad.Emissive_Min}
  n12 = ScalarParameter() {defaultValue=0.85, group=Ranked Reload, parameterName=RankedRelaoad.Emissive_Max}
  n13 = Power(
   Base=OneMinus(None=NamedRerouteUsage())
   Exp=ScalarParameter() {defaultValue=1, group=Ranked Reload, parameterName=RankedRelaoad.Emissive_FresnelExponent}
  )
  NamedRerouteDeclaration(None=Multiply(A=LinearInterpolate(A=@n8, B=@n9.RGB, Alpha=@n10), B=LinearInterpolate(A=@n11, B=@n12, Alpha=@n13))) {name=OUT_Emissive, nodeColor=<rgba 0.2891 1 0 1>}
  n14 = Subtract(A=NamedRerouteUsage(), B=NamedRerouteUsage())
  NamedRerouteDeclaration(None=LinearInterpolate(
    A=ScalarParameter() {group=Ranked Reload, parameterName=RankedReload.Roughness_Min}
    B=ScalarParameter() {defaultValue=1, group=Ranked Reload, parameterName=RankedReload.Roughness_Max}
    Alpha=Saturate(None=@n14)
   )) {name=OUT_Roughness, nodeColor=<rgba 1 0.4922 0 1>}
  n15 = Add(A=NamedRerouteUsage(), B=NamedRerouteUsage())
  NamedRerouteDeclaration(None=LinearInterpolate(
    A=ScalarParameter() {group=Ranked Reload, parameterName=RankedReload.Spec_Min}
    B=ScalarParameter() {defaultValue=1, group=Ranked Reload, parameterName=RankedReload.Spec_Max}
    Alpha=Saturate(None=@n15)
   )) {name=OUT_Spec, nodeColor=<rgba 1 0.4922 0 1>}
  NamedRerouteDeclaration(None=@n1.Fresnel) {name=Fresnel, nodeColor=<rgba 0.5469 0 1 1>}
  n16 = NamedRerouteUsage()
  n17 = SetMaterialAttributes!(
   MaterialAttributes=@n16
   Emissive Color=NamedRerouteUsage()
   Base Color=Multiply(A=NamedRerouteUsage(), B=ScalarParameter() {defaultValue=0.3, group=Ranked Reload, parameterName=RankedReload.BaseColor_GridAmount})
   Metallic=ScalarParameter() {defaultValue=0.7, group=Ranked Reload, parameterName=RankedReload.Metallic}
   Specular=NamedRerouteUsage()
   Roughness=NamedRerouteUsage()
   Opacity=ScalarParameter() {defaultValue=1, group=Ranked Reload, parameterName=RankedReload.Opacity, sortPriority=500}
  ) {attributeSetTypes [
    #1
    #2
    #3
    #4
    #5
    #6
   ]}
  n18 = ChannelMaskParameter(None=NamedRerouteUsage()) {group=Ranked Reload, maskChannel=Green, parameterName=RankedReload.MaskChannel, defaultValue=<rgba 0 1 0 0>}
  n19 = NamedRerouteDeclaration(None=@n18.R) {name=RankedReloadMask, nodeColor=<rgba 0 1 0.5781 1>}
  n20 = MaterialFunctionCall!(Material=SetMaterialAttributes!(MaterialAttributes), Tech Art Mask=NamedRerouteUsage()) {materialFunction=/BRCosmetics/Characters/Player/Female/Medium/Bodies/F_MED_CopperMoggy_Swan/Skins/Prime/Materials/MF_CopperMoggy_DigitalMosaic}
  NamedRerouteDeclaration(None=BlendMaterialAttributes(A=@n16, B=@n17, Alpha=Multiply(A=@n19, B=@n20.NoiseResult))) {name=OUT_Material, nodeColor=<rgba 0 1 0.4844 1>}
  n21 = Time()
  n22 = ScalarParameter() {defaultValue=0.3, group=Ranked Reload, parameterName=RankedReload.Wipe_Speed}
  n23 = Multiply(A=Frac(None=Multiply(A=@n21, B=@n22)), B=ScalarParameter() {defaultValue=100, group=Ranked Reload, parameterName=RankedReload.Wipe_Distance}) {constB=100}
  n24 = VectorParameter() {group=Ranked Reload, parameterName=RankedReload.Wipe_CenterPos, defaultValue=<rgba 0 0 108 1>}
  n25 = VertexInterpolator(VS=LocalPosition() {localOrigin=InstancePreSkinning}.XYZ)
  n26 = Add(A=Length(None=Subtract(A=@n24.RGB, B=@n25.PS)) {is=class MaterialExpressionLength}, B=NamedRerouteUsage())
  n27 = ScalarParameter() {defaultValue=8, group=Ranked Reload, parameterName=RankedReload.Wipe_Radius}
  n28 = ScalarParameter() {group=Ranked Reload, parameterName=RankedReload.Wipe_Hardness}
  NamedRerouteDeclaration(None=Power(
    Base=SphereMask(A=@n23, B=@n26, Radius=@n27, Hardness=@n28)
    Exp=ScalarParameter() {defaultValue=2, group=Ranked Reload, parameterName=RankedReload.Wipe_Exponent}
   )) {name=Wipe, nodeColor=<rgba 1 0.7031 0 1>}
  n29 = TextureCoordinate()
  n30 = ScalarParameter() {defaultValue=2, group=Ranked Reload, parameterName=RankedReload.Wipe_WarpNoiseTile}
  n31 = ChannelMaskParameter(None=TextureSample(UVs=Multiply(A=@n29, B=@n30)) {samplerType=SAMPLERTYPE_Masks, texture=/Game/Global/Textures/Default/Noise/T_Generic_Perlin}.RGB) {group=Ranked Reload, parameterName=RankedReload.Wipe_WarpNoiseChannel}
  n32 = ScalarParameter() {defaultValue=5, group=Ranked Reload, parameterName=RankedReload.Wipe_WarpNoiseStrength}
  n33 = ScalarParameter() {defaultValue=15, group=Ranked Reload, parameterName=RankedReload.Wipe_WarpGridStrength}
  n34 = NamedRerouteUsage()
  NamedRerouteDeclaration(None=Add(A=Multiply(A=@n31.R, B=@n32), B=Multiply(A=@n33, B=@n34))) {name=WipeWarp, nodeColor=<rgba 0.1016 1 0 1>}
  NamedRerouteDeclaration(None=@n7."Result (V3)") {name=ColorGrid, nodeColor=<rgba 1 0.3516 0 1>}
 ]
}
